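# Translucent unlit Niagara sprite material that shows either a flare or a bokeh shape, tinted by particle color.
# The rounded TextureLerp dynamic parameter picks between a blend of cross and spikeball flares with boosted color, and bokeh with unboosted color.
# Color can blend toward an MF_FX_EmissiveAdjustments result by parameter.
# Opacity combines the shape, particle alpha, depth fade and near-camera fade.
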Material /Game/Environments/Ch6/Hermes/MagicForest/Materials/M_FlareLerpBokeh
graph {
 properties {
  bAutomaticallySetUsageInEditor=false
  bUsedWithNiagaraSprites=true
  blendMode=BLEND_Translucent
  shadingModel=MSM_Unlit
  usedShadingModels=Unlit
 }
 nodes [
  DynamicParameter() {paramNames [TextureLerp,Param2,Param3,Param4]}
  n1 = ParticleColor()
  n2 = Round(None=DynamicParameter() {paramNames [TextureLerp,Param2,Param3,Param4]}.TextureLerp)
  n3 = LinearInterpolate(A=Multiply(A=@n1.RGB, B=ScalarParameter() {defaultValue=5}), B=@n1.RGB, Alpha=@n2)
  LinearInterpolate(
   A=@n3
   B=MaterialFunctionCall(Emissive=@n3) {materialFunction=&MF_FX_EmissiveAdjustments}.Result
   Alpha=ScalarParameter() {parameterName=USE EMISSIVE ADJUSTMENT}
  )
  n4 = LinearInterpolate(
   A=LinearInterpolate(
    A=TextureSample() {samplerType=SAMPLERTYPE_Grayscale, texture=/Game/Effects/Fort_Effects/Textures/Flares/T_Cross_Flare}.G
    B=TextureSample() {texture=/Game/Effects/Fort_Effects/Textures/Flares/T_Flare_Spikeball}.G
   )
   B=TextureSample() {samplerType=SAMPLERTYPE_LinearColor, texture=/Game/TimeOfDay/TODM/BR/s11/ExposureAssets/FN_BokehS11}.G
   Alpha=@n2
  )
  Multiply(
   A=DepthFade(Opacity=Multiply(A=@n1.A, B=@n4))
   B=MaterialFunctionCall() {materialFunction=&Near_Camera_Fade}.Result
  )
 ]
}
